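; PCG sample subgraph that derives grammar module information from meshes supplied as points or attribute sets.
; It renames the configurable mesh attribute to Mesh, filters by data type, and converts attribute sets to points so both feed Bounds From Mesh.
; It derives Size from twice the X extent, saves ExtentsZ for scaling, and sets PivotOffset to the negated local center.
; The result is output as an attribute set.
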(graph ExtractMeshInfo
  (params (MeshAttributeName :type Name :default Mesh))
  (inputs (MeshInfo :type Point%20|%20Attribute%20Set))
  (bind AttributeRename_8 (Attribute%20Rename :AttributeToRename @LastCreated :NewAttributeName Mesh :AttributeToRename-param MeshAttributeName))
  (bind BoundsFromMesh_0 (Bounds%20From%20Mesh :MeshAttribute Mesh))
  (bind CreateAttribute_3 (Create%20Constant :AttributeTypes (struct :DoubleValue 2.0)))
  (bind AttributeMathsOp_2 (Attribute%20Maths%20Op :InputSource1 $Extents.X :Operation Multiply :OutputTarget Size)) ; Extract Extents.X times 2\nfor module size
  (bind CopyAttributes_6 (Copy%20Attributes :InputSource $Extents.Z :OutputTarget ExtentsZ)) ; Save Extents.Z for scaling
  (bind CreateAttribute_5 (Create%20Constant :AttributeTypes (struct :DoubleValue -1.0)))
  (bind AttributeMathsOp_4 (Attribute%20Maths%20Op :InputSource1 $LocalCenter :Operation Multiply :OutputTarget PivotOffset)) ; -LocalCenter is the PivotOffset
  (bind PointToAttributeSet_10 (Point%20To%20Attribute%20Set))
  ;[0.15 0.15 0.15 0.5] Supporting both points and
  ; attribute sets
  (bind FilterDataByType_3 (Filter%20Data%20By%20Type :TargetType (struct :Ids [(struct :Struct /Script/PCG.PCGDataTypeInfoParam)])))
  (bind AttributeSetToPoint_2 (Attribute%20Set%20To%20Point))
  (bind FilterDataByType_5 (Filter%20Data%20By%20Type :TargetType (struct :Ids [(struct :Struct /Script/PCG.PCGDataTypeInfoPoint)])))
  (connect BoundsFromMesh_0 Out AttributeMathsOp_2 InA)
  (connect CreateAttribute_3 Out AttributeMathsOp_2 InB)
  (connect CopyAttributes_6 Out AttributeMathsOp_4 InA)
  (connect CreateAttribute_5 Out AttributeMathsOp_4 InB)
  (connect DefaultInputNode MeshInfo AttributeRename_8 In)
  (connect FilterDataByType_3 InsideFilter AttributeSetToPoint_2 In)
  (connect AttributeSetToPoint_2 Out BoundsFromMesh_0 In)
  (connect FilterDataByType_5 InsideFilter BoundsFromMesh_0 In)
  (connect AttributeMathsOp_2 Out CopyAttributes_6 Source)
  (connect AttributeMathsOp_2 Out CopyAttributes_6 Target)
  (connect AttributeRename_8 Out FilterDataByType_3 In)
  (connect AttributeRename_8 Out FilterDataByType_5 In)
  (connect AttributeMathsOp_4 Out PointToAttributeSet_10 In)
  (output Out PointToAttributeSet_10))
