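; PCG sample graph that builds modular buildings from splines using shape grammars, taking an input spline or fetching one with a default height.
; Grammar-driven Duplicate Cross-Sections stacks floors, whose edge segments pick grammars by floor type and criteria before subdividing into column, wall and window modules.
; Module info, enriched by an ExtractMeshInfo subgraph, drives slicing and pivot/scale correction for mesh spawning.
; Toggles control spawning and colored floor/module debug views.
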(graph PCG_BuildingSample
  (params (DebugFloors :type Bool :default true) (DebugModules :type Bool :default true) (SpawnMeshes :type Bool :default true) (BuildingHeight :type Double :default 1.5e+03) (ModuleInfo :type (Array (Struct /Script/PCG.PCGSubdivisionSubmodule)) :default [(struct :Symbol C :DebugColor (1.0 0.0 0.0 1.0)) (struct :Symbol W :bScalable true :DebugColor (0.0 1.0 0.0 1.0)) (struct :Symbol W1 :bScalable true :DebugColor (0.0 0.0 1.0 1.0)) (struct :Symbol W2 :bScalable true :DebugColor (1.0 0.0 1.0 1.0))]) (MeshInfo :type (Array (Object /Script/Engine.StaticMesh)) :default [/PCG/SampleContent/Grammar/Assets/PCG_Column.PCG_Column /PCG/SampleContent/Grammar/Assets/PCG_Wall.PCG_Wall /PCG/SampleContent/Grammar/Assets/PCG_Window.PCG_Window /PCG/SampleContent/Grammar/Assets/PCG_HoleWindow.PCG_HoleWindow]))
  (inputs (SplineInput :type Spline))
  (bind FilterDataByType_6 (Filter%20Data%20By%20Type :TargetType (struct :Ids [(struct :Struct /Script/PCG.PCGDataTypeInfoSpline)])))
  (bind VolumeSlicer_5 (Duplicate%20Cross-Sections :bExtrudeVectorAsAttribute true :bOutputDebugColorAttribute true :ExtrudeVectorAttribute @Data.ExtrudeVector :GrammarSelection (struct :bGrammarAsAttribute true :GrammarAttribute @Data.Grammar) :ModulesInfo [(struct :Symbol MainFloor :Size 3e+02 :DebugColor (1.0 0.0 0.0 1.0)) (struct :Symbol Intermediate :Size 2.5e+02 :DebugColor (0.0 1.0 0.0 1.0))])) ; Grammar selection \nfor the floors
  (bind SplineDirection_0 (Spline%20Direction :Operation ForceCounterClockwise))
  (bind SplineToSegment_25 (Spline%20to%20Segment)) ; Transform Spline \ninto points \nrepresenting the segments
  (bind AttributeMathsOp_35 (Attribute%20Maths%20Op :InputSource1 $BoundsMax.Z :InputSource2 @Data.Size.Z :Operation Set)) ; Assign Size.Z to BoundsMax.Z
  (bind SelectGrammar_26 (Select%20Grammar :Criteria [(struct :Key MainFloor :Comparator GreaterThan :FirstValue 499.0 :Grammar "[C][W,W1]*[C]") (struct :Key MainFloor :Comparator Select :Grammar "[C][W]*[C]") (struct :Key Intermediate :Comparator GreaterThan :FirstValue 5e+02 :Grammar "[C][W1]*[W2][W1]*[C]") (struct :Key Intermediate :Comparator Select :Grammar "[C][W1]*[C]")] :KeyAttribute @Data.Symbol)) ; Select grammar \nusing some criteria
  (bind Subgraph_22 (Subgraph :Graph /PCG/SampleContent/Grammar/Graphs/ExtractMeshInfo.ExtractMeshInfo))
  (bind SegmentSlicer_27 (Subdivide%20Segment :bForwardAttributesFromModulesInfo true :bModuleInfoAsInput true :bOutputDebugColorAttribute true :GrammarSelection (struct :bGrammarAsAttribute true :GrammarAttribute Grammar) :ModulesInfoAttributeNames (struct :bProvideScalable true :ScalableAttributeName bScalable :bProvideDebugColor true))) ; Actual module selection using the grammar\nModule info is forwarded to the output\n(Match and Set done inside the node, with \nthe option Forward Attributes from Module Info)
  (bind Branch_44 (Branch :bOutputToB-param SpawnMeshes))
  (bind StaticMeshSpawner_39 (Static%20Mesh%20Spawner :MeshSelectorParameters (class:/Script/PCG.PCGMeshSelectorByAttribute :AttributeName Mesh) :MeshSelectorType /Script/PCG.PCGMeshSelectorByAttribute))
  ;[0.04214 0.08074 0.15] Extract Spline and add attributes for building size if we have no input
  (bind GetSplineData_0 (Get%20Spline%20Data))
  (bind MakeVectorAttribute_46 (Make%20Vector%20Attribute :OutputType Vector :Z BuildingHeight))
  (bind AddAttribute_48 (Add%20Attribute :OutputTarget @Data.ExtrudeVector))
  ; If coming from an input, use this one. You can check the
  ; Primitive Cross Section usage in the PCG Geometry Interop plugin content
  (bind DataNum_0 (Data%20Count))
  (bind Select_2 (Select))
  ;[0.15 0.15 0.15 0.5] Add Grammar attribute for floor selection on Data domain
  (bind AddAttribute_3 (Add%20Attribute :AttributeTypes (struct :Type String :VectorValue (0.0 0.0 1.5e+03) :StringValue "[MainFloor][Intermediate]*") :OutputTarget @Data.Grammar))
  ;[0.06897 0.15 0.06193] Debug
  (bind CreateSurfaceFromSpline_12 (Create%20Surface%20From%20Spline))
  (bind SurfaceSampler_13 (Surface%20Sampler :bUnbounded true :PointExtents (1e+01 1e+01 1e+01) :PointsPerSquaredMeter 5.0 :PointSteepness 0.5))
  (bind Branch_30 (Branch :bOutputToB-param DebugFloors))
  (bind Branch_33 (Branch :bOutputToB-param DebugModules))
  (bind Gather_29 (Gather))
  (bind CopyAttributes_9 (Copy%20Attributes :InputSource @Data.DebugColor :OutputTarget $Color))
  (bind CreateAttribute_37 (Create%20Constant :AttributeTypes (struct :DoubleValue 0.99)))
  (bind BoundsModifier_36 (Bounds%20Modifier)) ; Small Bounds modifier to see the modules
  (bind Debug_11 (Debug))
  ;[0.1164 0.15 0.03476 0.5] This could come from a data table
  (bind AddAttribute_2 (Add%20Attribute :OutputTarget Mesh :Attributes MeshInfo :In ModuleInfo))
  ;[0.1458 1 0.9901] Post-processing to make it match with incoming meshes
  (bind AttributeMathsOp_43 (Attribute%20Maths%20Op :InputSource1 $Extents.Z :InputSource2 ExtentsZ :Operation Divide :OutputTarget $Scale.Z)) ; Scale the mesh in Z\naccordingly
  (bind AttributeMathsOp_40 (Attribute%20Maths%20Op :InputSource1 PivotOffset :InputSource2 $Scale :Operation Multiply)) ; PivotOffset needs to be \nscaled too
  (bind AttributeVectorOp_42 (Attribute%20Vector%20Op :InputSource1 PivotOffset :InputSource2 $Transform :Operation TransformDirection)) ; And rotated
  (bind AttributeMathsOp_41 (Attribute%20Maths%20Op :InputSource1 $Position :InputSource2 PivotOffset)) ; And added to the position
  (connect Select_2 Out AddAttribute_3 In)
  (connect MakeVectorAttribute_46 Out AddAttribute_48 Attributes)
  (connect GetSplineData_0 Out AddAttribute_48 In)
  (connect SplineToSegment_25 Out AttributeMathsOp_35 InA)
  (connect SplineToSegment_25 Out AttributeMathsOp_35 InB)
  (connect AttributeMathsOp_43 Out AttributeMathsOp_40 InA)
  (connect AttributeMathsOp_43 Out AttributeMathsOp_40 InB)
  (connect AttributeVectorOp_42 Out AttributeMathsOp_41 InA)
  (connect AttributeVectorOp_42 Out AttributeMathsOp_41 InB)
  (connect SegmentSlicer_27 Out AttributeMathsOp_43 InA)
  (connect SegmentSlicer_27 Out AttributeMathsOp_43 InB)
  (connect AttributeMathsOp_40 Out AttributeVectorOp_42 In)
  (connect AttributeMathsOp_40 Out AttributeVectorOp_42 Transform)
  (connect CreateAttribute_37 Out BoundsModifier_36 BoundsMax)
  (connect CreateAttribute_37 Out BoundsModifier_36 BoundsMin)
  (connect CopyAttributes_9 Out BoundsModifier_36 In)
  (connect SurfaceSampler_13 Out Branch_30 In)
  (connect SegmentSlicer_27 Out Branch_33 In)
  (connect AttributeMathsOp_41 Out Branch_44 In)
  (connect Gather_29 Out CopyAttributes_9 Source)
  (connect Gather_29 Out CopyAttributes_9 Target)
  (connect VolumeSlicer_5 Out CreateSurfaceFromSpline_12 In)
  (connect DefaultInputNode SplineInput DataNum_0 In)
  (connect BoundsModifier_36 Out Debug_11 In)
  (connect AddAttribute_3 Out FilterDataByType_6 In)
  (connect Branch_30 Output%20B Gather_29 In)
  (connect Branch_33 Output%20B Gather_29 In)
  (connect SelectGrammar_26 Out SegmentSlicer_27 In)
  (connect Subgraph_22 Out SegmentSlicer_27 ModulesInfo)
  (connect DataNum_0 Out Select_2 bUseInputB)
  (connect AddAttribute_48 Out Select_2 Input%20A)
  (connect DefaultInputNode SplineInput Select_2 Input%20B)
  (connect AttributeMathsOp_35 Out SelectGrammar_26 In)
  (connect VolumeSlicer_5 Out SplineDirection_0 In)
  (connect SplineDirection_0 Out SplineToSegment_25 In)
  (connect Branch_44 Output%20B StaticMeshSpawner_39 In)
  (connect AddAttribute_2 Out Subgraph_22 MeshInfo)
  (connect CreateSurfaceFromSpline_12 Out SurfaceSampler_13 Surface)
  (connect FilterDataByType_6 InsideFilter VolumeSlicer_5 In))
