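; PCG sample graph that slices splines into fence, corner-sized "C" and door modules using a grammar string, visualized as colored debug cubes.
; Spline data gets a Grammar attribute, which Subdivide Spline reads alongside module definitions embedded in the node.
; Segment bounds are tweaked to separate modules, then divided by cube mesh extents to set scale.
; The mesh spawner packs density, extents and debug color per instance, mimicking a persistent Debug node.
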(graph PCG_SplineSlicer
  (params (Grammar :type String :default "[Fence]*[C]2[Door][C]2[Fence]*"))
  (bind GetSplineData_1 (Get%20Spline%20Data))
  (bind AddAttribute_0 (Add%20Attribute :OutputTarget @Data.Grammar :Attributes Grammar))
  (bind SplineSlicer_2 (Subdivide%20Spline :bOutputDebugColorAttribute true :GrammarSelection (struct :bGrammarAsAttribute true :GrammarString "[Fence]*[C]2[Door][C]2[Fence]*" :GrammarAttribute @Data.Grammar) :ModulesInfo [(struct :Symbol Fence :bScalable true :DebugColor (1.0 0.0 0.0 1.0)) (struct :Symbol Door :Size 3e+02 :DebugColor (0.0 1.0 0.0 1.0)) (struct :Symbol C :Size 5e+01 :DebugColor (0.0 0.0 1.0 1.0))])) ; Module info is \nembedded in the node
  (bind BoundsModifier_3 (Bounds%20Modifier :BoundsMax (0.9 1.0 1.0) :BoundsMin (0.9 1.0 1.0))) ; Modify bounds to see the\ndifferent modules
  ;[0.15 0.15 0.15 0.5] Mimic Debug node (but persistent)
  (bind PointFromMesh_0 (Point%20From%20Mesh :Mesh /PCG/DebugObjects/PCG_Cube.PCG_Cube))
  (bind AttributeMathsOp_1 (Attribute%20Maths%20Op :InputSource1 $Extents :InputSource2 $Extents :Operation Divide :OutputTarget $Scale))
  (bind StaticMeshSpawner_0 (Static%20Mesh%20Spawner :bApplyMeshBoundsToPoints false :InstanceDataPackerParameters (class:/Script/PCG.PCGInstanceDataPackerByAttribute :AttributeSelectors [$Density $Extents DebugColor]) :InstanceDataPackerType /Script/PCG.PCGInstanceDataPackerByAttribute :MeshSelectorParameters (class:/Script/PCG.PCGMeshSelectorWeighted :MeshEntries [(struct :Descriptor (struct :StaticMesh /PCG/DebugObjects/PCG_Cube.PCG_Cube :OverrideMaterials [/PCG/DebugObjects/PCG_DebugMaterial_Color.PCG_DebugMaterial_Color] :bCanEverAffectNavigation true))])))
  (connect GetSplineData_1 Out AddAttribute_0 In)
  (connect BoundsModifier_3 Out AttributeMathsOp_1 InA)
  (connect PointFromMesh_0 Out AttributeMathsOp_1 InB)
  (connect SplineSlicer_2 Out BoundsModifier_3 In)
  (connect AddAttribute_0 Out SplineSlicer_2 In)
  (connect AttributeMathsOp_1 Out StaticMeshSpawner_0 In))
